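# Translucent unlit hologram material for Niagara ribbons, sprites and mesh particles, in a cosmetic shoe's FX folder.
# MF_3DHB_HologramFX runs on screen-aligned UVs centered on the actor, its opacity blends two colors, and a dynamic parameter scales its scanline glitch.
# A scalar blends between four-edge UV fades and a particle mask texture to shape opacity, alongside particle alpha.
# Depth fade softens intersections on all shading paths except mobile.
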Material /CosmeticShoes/Assets/FORT_Shoes/Shoe_SashSkewerPollen/FX/M_HoloRibbon
graph {
 properties {
  bAutomaticallySetUsageInEditor=false
  bUsedWithNiagaraMeshParticles=true
  bUsedWithNiagaraRibbons=true
  bUsedWithNiagaraSprites=true
  blendMode=BLEND_Translucent
  shadingModel=MSM_Unlit
  usedShadingModels=Unlit
 }
 nodes [
  MaterialFunctionCall(Emissive=NamedRerouteUsage()) {materialFunction=&MF_FX_EmissiveAdjustments}
  n1 = MaterialFunctionCall!(Center Point World Position=ActorPositionWS()) {materialFunction=/Engine/Functions/Engine_MaterialFunctions02/UVs/WorldSpaceAlignedScreenCoordinates}
  n2 = MaterialFunctionCall!(
   Time
   MeshScaleIn
   UVForScanline=@n1.Offset and Scaled UVS
   CustomFlicker
   UseCustomFlicker
   UVForStroke=@n1.Offset and Scaled UVS
   UVForGlitch=@n1.Offset and Scaled UVS
   BaseUVIn=@n1.Offset and Scaled UVS
   EmissiveIn
   OpacityIn
   WPOIn
  ) {materialFunction=/Game/Tools/3DHB/Shader/Base/MF_3DHB_HologramFX}
  NamedRerouteDeclaration(None=Multiply(A=@n2.UseForFX_ScanlineOut, B=DynamicParameter() {paramNames [Glitch,Param2,Param3,Param4]}.Glitch)) {name=GlitchOut, nodeColor=<rgba 1 0.9844 0 1>}
  NamedRerouteDeclaration(None=@n2.OpacityOut) {name=AlphaOut, nodeColor=<rgba 0.3828 0 1 1>}
  NamedRerouteDeclaration(None=LinearInterpolate(
    A=Constant4Vector() {constant=<rgba 0.2034 0.0044 1 1>}.RGBA
    B=Constant4Vector() {constant=<rgba 0.046 0.2544 1 1>}.RGBA
    Alpha=@n2.OpacityOut
   )) {name=EmissiveOut, nodeColor=<rgba 0 0.9766 1 1>}
  n3 = NamedRerouteUsage()
  n4 = ScalarParameter() {
   defaultValue=0.05
   desc=Length in uv space of the leading edge fading in.
   group="4) UV Fade"
   parameterName=Top Edge Fade Distance
   sortPriority=0
  }
  n5 = Max(A=@n4) {constB=0.0001}
  n6 = TextureCoordinate()
  n7 = ComponentMask(None=@n6) {g=true}
  n8 = Clamp(None=Multiply(A=@n7, B=Divide(B=@n5)))
  n9 = ScalarParameter() {defaultValue=2.2, desc=Exponent of the leading edge fade in., group="4) UV Fade", parameterName=Top Edge Fade Power, sortPriority=1}
  n10 = ScalarParameter() {
   defaultValue=0.1
   desc=Length in uv space of the trailing edge fading out.
   group="4) UV Fade"
   parameterName=Bottom Edge Fade Distance
   sortPriority=2
  }
  n11 = Max(A=@n10) {constB=0.0001}
  n12 = Clamp(None=Multiply(A=OneMinus(None=@n7), B=Divide(B=@n11)))
  n13 = ScalarParameter() {defaultValue=2.2, desc=Exponent of the trailing edge fade out., group="4) UV Fade", parameterName=Bottom Edge Fade Power, sortPriority=3}
  n14 = Multiply(A=Clamp(None=Power(Base=@n8, Exp=@n9) {constExponent=2.2}), B=Clamp(None=Power(Base=@n12, Exp=@n13) {constExponent=2.2}))
  n15 = ScalarParameter() {
   defaultValue=0.05
   desc=Length in uv space of the leading edge fading in.
   group="4) UV Fade"
   parameterName=Leading Edge Fade Distance
   sortPriority=0
  }
  n16 = Max(A=@n15) {constB=0.0001}
  n17 = ComponentMask(None=@n6) {r=true}
  n18 = Clamp(None=Multiply(A=@n17, B=Divide(B=@n16)))
  n19 = ScalarParameter() {defaultValue=2.2, desc=Exponent of the leading edge fade in., group="4) UV Fade", parameterName=Leading Edge Fade Power, sortPriority=1}
  n20 = ScalarParameter() {
   defaultValue=0.1
   desc=Length in uv space of the trailing edge fading out.
   group="4) UV Fade"
   parameterName=Trailing Edge Fade Distance
   sortPriority=2
  }
  n21 = Max(A=@n20) {constB=0.0001}
  n22 = Clamp(None=Multiply(A=OneMinus(None=@n17), B=Divide(B=@n21)))
  n23 = ScalarParameter() {defaultValue=2.2, desc=Exponent of the trailing edge fade out., group="4) UV Fade", parameterName=Trailing Edge Fade Power, sortPriority=3}
  n24 = Min(A=Clamp(None=Power(Base=@n18, Exp=@n19) {constExponent=2.2}), B=Clamp(None=Power(Base=@n22, Exp=@n23) {constExponent=2.2}))
  n25 = TextureSampleParameter2D() {
   parameterName=ParticleMaskTexture
   samplerType=SAMPLERTYPE_LinearGrayscale
   texture=/Game/Characters/Player/Female/Medium/Bodies/F_MED_Bistro_Astronaut/FX/T_Water_Splash_Alpha_02
  }
  n26 = LinearInterpolate(
   A=NamedRerouteDeclaration(None=Multiply(A=@n14, B=@n24)) {name=RibbonMaskOut, nodeColor=<rgba 1 0.7031 0 1>}
   B=@n25.RGB
   Alpha=ScalarParameter() {parameterName=Ribbon Or Particle Mask}
  )
  n27 = Multiply(A=Multiply(A=Multiply(A=@n3, B=@n26), B=ParticleColor().A), B=NamedRerouteUsage())
  n28 = DepthFade(
   Opacity=@n27
   FadeDistance=ScalarParameter() {defaultValue=100, desc=Depth fade distance., group="4) Fade", parameterName=Depth Fade Distance}
  )
  ShadingPathSwitch(Default=@n28, Deferred=@n28, Forward=@n28, Mobile=@n27)
 ]
}
